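# Translucent UI material, named as the background for the vehicle HUD's fuel warning message, with a panning two-color triangle pattern.
# Pixel-size aspect correction feeds rotated, panned and tiled UI UVs into a triangle grid texture, whose red channel blends Color1 and Color2.
# A separate branch multiplies scaled and rotated UV gradients and raises them to an Edge Falloff power to form a soft edge mask.
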Material /Game/Athena/HUD/Vehicles/Materials/M_UI_FuelWarningMessageBG
graph {
 properties {blendMode=BLEND_Translucent, materialDomain=MD_UI, pixelDepthOffsetMode=PDOM_Legacy, usedShadingModels=Default Lit}
 nodes [
  Constant2Vector() {g=-0.5}
  Constant() {r=1}
  n1 = MaterialFunctionCall() {materialFunction=&GetUserInterfaceUV}
  n2 = MaterialFunctionCall(Float2=@n1.Pixel Size) {materialFunction=&BreakOutFloat2Components}
  n3 = AppendVector(A=Divide(A=@n2.R, B=Max(A=@n2.G)), B=Constant() {r=1})
  n4 = Panner(
   Coordinate=Rotator(Coordinate=Multiply(A=@n3, B=@n1.Normalized UV), Time=ScalarParameter() {defaultValue=-0.0278, parameterName=Tri Angle}) {speed=6.2832}
   Speed=AppendVector(A=ScalarParameter() {defaultValue=0.5, parameterName=PanSpeed}, B=Constant())
  )
  n5 = ScalarParameter() {defaultValue=0.5, parameterName=Tri Tile Amount}
  LinearInterpolate(
   A=VectorParameter() {parameterName=Color1, defaultValue=<rgba 0.9617 0 1 0>}.RGB
   B=VectorParameter() {parameterName=Color2, defaultValue=<rgba 0.086 0 1 0>}.RGB
   Alpha=TextureSample(UVs=Multiply(A=@n4, B=@n5)) {texture=/Game/Global/Textures/Default/Masks_and_Gradients/T-TriangleGrid}.R
  )
  n6 = AppendVector(A=ScalarParameter() {defaultValue=2, parameterName=UVScale}, B=Constant() {r=1})
  n7 = TextureCoordinate()
  MaterialFunctionCall!(Texture Scale=@n6, UVs=@n7, Offset XY Texturesize Z=Constant3Vector() {constant=<rgba 0 0.5 256 0>}) {materialFunction=/Engine/Functions/Engine_MaterialFunctions01/Texturing/ScaleUVsAroundPoint}
  n8 = Multiply(A=@n7, B=@n6)
  n9 = ComponentMask(None=@n8) {r=true}
  n10 = ScalarParameter() {defaultValue=2, parameterName=angle1}
  n11 = ScalarParameter() {parameterName=angle2}
  n12 = ComponentMask(None=MaterialFunctionCall(UVs=@n8, Rotation Center=AppendVector(A=@n10, B=@n11), "Rotation Angle (0-1)"=Constant() {r=-0.25}) {materialFunction=&CustomRotator}.Rotated Values) {g=true}
  Multiply(A=Power(Base=Multiply(A=@n9, B=@n12), Exp=ScalarParameter() {defaultValue=5, parameterName=Edge Falloff}))
  OneMinus(None=@n9)
 ]
}
