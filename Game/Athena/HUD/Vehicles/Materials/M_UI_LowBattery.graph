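# Translucent UI material for a vehicle HUD low battery warning, drawing an icon with animated diamond pulses around it.
# The icon texture is stretched, offset and scaled about its center, then optionally SDF-thresholded with SmoothStep.
# Raw texture and thresholded masks blend two icon colors over the pulse color.
# Two MF_DiamondPulse rings with separate end scales and widths are max-combined with the icon mask, apparently as opacity.
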Material /Game/Athena/HUD/Vehicles/Materials/M_UI_LowBattery
graph {
 properties {blendMode=BLEND_Translucent, materialDomain=MD_UI, pixelDepthOffsetMode=PDOM_Legacy, usedShadingModels=Default Lit}
 nodes [
  n1 = VectorParameter() {group=Icon, parameterName=IconStretch, defaultValue=<rgba 1 1 0 0>}
  n2 = Multiply(A=TextureCoordinate(), B=AppendVector(A=@n1.R, B=@n1.R))
  n3 = AppendVector(A=@n1.B, B=@n1.B)
  n4 = TextureSampleParameter2D(UVs=MaterialFunctionCall(UVs=Add(A=@n2, B=@n3), Texture Scale=ScalarParameter() {defaultValue=0.6, group=Icon, parameterName=IconScale}) {materialFunction=&ScaleUVsByCenter}.UVs) {group=Icon, parameterName=IconTexture, samplerSource=SSM_Clamp_WorldGroupSettings, texture=/Game/Athena/HUD/Materials/T_UI_LowBattery}
  n5 = ScalarParameter() {defaultValue=0.2, group=Icon, parameterName=Edge}
  n6 = Multiply(A=@n5, B=ScalarParameter() {defaultValue=0.1, group=Icon, parameterName=EdgeSoftness}) {constB=0.1}
  n7 = StaticSwitchParameter(
   True=MaterialFunctionCall(Alpha=@n4.R, Min=@n5, Max=Add(A=@n5, B=@n6)) {materialFunction=&SmoothStep}.Result
   False=@n4.R
  ) {defaultValue=true, group=Icon, parameterName=IsSDFTexture}
  n8 = Max(A=@n4.R, B=@n7)
  LinearInterpolate(
   A=VectorParameter() {parameterName=PulseColor, defaultValue=<rgba 1 0.4942 0 1>}.RGB
   B=LinearInterpolate(
    A=VectorParameter() {parameterName=Color2, defaultValue=<rgba 0.6562 0 0 1>}.RGB
    B=VectorParameter() {parameterName=Color1, defaultValue=<rgba 1 0.0469 0 1>}.RGB
    Alpha=@n7
   )
   Alpha=@n8
  )
  n9 = ScalarParameter() {group=PulseStyle, parameterName="ScaleStart (S)"}
  n10 = ScalarParameter() {defaultValue=0.75, group=PulseStyle, parameterName=ScaleEnd}
  n11 = ScalarParameter() {defaultValue=0.3, group=PulseStyle, parameterName="WidthEnd (S)"}
  n12 = Constant() {r=0.125}
  n13 = ScalarParameter() {group=PulseAnim, parameterName="TimeOffset (S)"}
  n14 = ScalarParameter() {group=PulseAnim, parameterName="ManualTime (S)"}
  n15 = ScalarParameter() {defaultValue=0.6, group=PulseStyle, parameterName=ScaleEnd2}
  n16 = ScalarParameter() {defaultValue=0.01, group=PulseStyle, parameterName=WidthEnd2}
  n17 = ScalarParameter() {defaultValue=0.2605, group=PulseStyle, parameterName="PulseAnimFalloff (S)"}
  Max(
   A=@n8
   B=Max(
    A=MaterialFunctionCall!(
     PulseSpeed
     ScaleStart=@n9
     ScaleEnd=@n10
     WidthStart
     WidthEnd=@n11
     PulseAnimFalloff
     Rotation=@n12
     TimeOffset=@n13
     ManualTime=@n14
     OpacityStart
     OpacityEnd
     TimeSpeed
     TexCoord
    ) {materialFunction=/Game/Athena/HUD/LTM/Infiltration/Materials/MF_DiamondPulse}.Result
    B=MaterialFunctionCall(ScaleStart=@n9, ScaleEnd=@n15, WidthEnd=@n16, PulseAnimFalloff=@n17, Rotation=@n12, TimeOffset=@n13, ManualTime=@n14) {materialFunction=&MF_DiamondPulse}.Result
   )
  )
  ScalarParameter() {defaultValue=0.1, group=PulseStyle, parameterName=TimeOffset}
 ]
}
