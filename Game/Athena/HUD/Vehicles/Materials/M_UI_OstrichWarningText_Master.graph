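# Translucent UI master material for the Ostrich vehicle warning text HUD, drawing a procedural shape without textures.
# An IsOutline static switch picks a rounded-box outline or a soft exponential radial gradient.
# The outline masks an inner rounded box out of an outer one, both built from 9-slice UVs with shared corner curves.
# Outline mode uses Filled_Color, gradient mode blends from Falloff_Color to Filled_Color, and Final_Alpha scales opacity.
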Material /Game/Athena/HUD/Vehicles/Materials/M_UI_OstrichWarningText_Master
graph {
 properties {blendMode=BLEND_Translucent, materialDomain=MD_UI, pixelDepthOffsetMode=PDOM_Legacy, usedShadingModels=Default Lit}
 nodes [
  n1 = ConstantBiasScale(None=MaterialFunctionCall() {materialFunction=&GetUserInterfaceUV}."9-Slice UV (No Tiling)") {bias=-0.5, scale=2}
  n2 = Abs(None=Divide(A=@n1, B=ScalarParameter() {defaultValue=1, parameterName=Size}))
  n3 = OneMinus(None=AppendVector(A=ScalarParameter() {defaultValue=0.25, parameterName=Curve_X}, B=ScalarParameter() {defaultValue=0.25, parameterName=Curve_Y}))
  n4 = Subtract(B=@n3)
  n5 = Saturate(None=Divide(A=Subtract(A=@n2, B=@n3) {constB=0.25}, B=@n4))
  n6 = Saturate(None=DotProduct(A=@n5, B=@n5))
  n7 = Multiply(
   A=Subtract(A=OneMinus(None=@n6), B=ScalarParameter() {defaultValue=0.005, parameterName=Subtract})
   B=ScalarParameter() {defaultValue=4, parameterName=Hardness}
  )
  n8 = Abs(None=Divide(A=@n1, B=ScalarParameter() {defaultValue=1, parameterName=SizeInner}))
  n9 = Saturate(None=Divide(A=Subtract(A=@n8, B=@n3) {constB=0.25}, B=@n4))
  n10 = DotProduct(A=@n9, B=@n9)
  n11 = Subtract(A=OneMinus(None=Saturate(None=@n10)), B=ScalarParameter() {defaultValue=0.005, parameterName=SubtractInner})
  n12 = Saturate(None=Multiply(A=Saturate(None=@n11), B=ScalarParameter() {defaultValue=4, parameterName=HardnessInner}))
  n13 = Saturate(None=DotProduct(A=Saturate(None=@n7), B=OneMinus(None=@n12)))
  n14 = Saturate(None=MaterialFunctionCall(
    Radius=ScalarParameter() {defaultValue=0.5, group=Treatment, parameterName=Outside_Radius}
    Density=ScalarParameter() {defaultValue=16, group=Treatment, parameterName=Outside_Density}
   ) {materialFunction=&RadialGradientExponential}.RadialGradientExponential)
  Saturate(None=Multiply(
    A=StaticSwitchParameter(True=@n13, False=@n14) {parameterName=IsOutline}
    B=ScalarParameter() {defaultValue=1, group=Global, parameterName=Final_Alpha}
   ))
  StaticSwitchParameter(
   True=VectorParameter() {parameterName=Filled_Color, defaultValue=<rgba 0 1 0 1>}.RGB
   False=LinearInterpolate(
    A=VectorParameter() {group=Color, parameterName=Falloff_Color, defaultValue=<rgba 1 0 0 1>}.RGB
    B=VectorParameter() {group=Color, parameterName=Filled_Color, defaultValue=<rgba 0 1 0 1>}.RGB
    Alpha=@n14
   )
  ) {parameterName=IsOutline}
 ]
}
