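# Additive UI material that draws a glowing, fillable ring arc for the Valet vehicle HUD meter.
# Two exponential radial gradients are subtracted to form the ring, then masked by the Max of two smoothstepped angle gradients, one inverted.
# PercentFill lerps both angular thresholds and their edge softness, setting the arc's extent.
# Panning blurred noise modulates that mask, which blends between the outer and main glow colors.
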Material /Game/Athena/HUD/Vehicles/Materials/M_UI_ValetMeterGlow
graph {
 properties {blendMode=BLEND_Additive, materialDomain=MD_UI, pixelDepthOffsetMode=PDOM_Legacy, usedShadingModels=Default Lit}
 nodes [
  n1 = ScalarParameter() {defaultValue=0.5, parameterName="Radius (S)"}
  n2 = ScalarParameter() {defaultValue=17, parameterName=Density}
  n3 = ScalarParameter() {defaultValue=0.05, parameterName=InnerRadius}
  n4 = Subtract(
   A=MaterialFunctionCall(Radius=@n1, Density=@n2) {materialFunction=&RadialGradientExponential}.RadialGradientExponential
   B=MaterialFunctionCall(Radius=Subtract(A=@n1, B=@n3), Density=@n2) {materialFunction=&RadialGradientExponential}.RadialGradientExponential
  )
  n5 = ScalarParameter() {defaultValue=-0.75, parameterName=AngleRotation}
  n6 = OneMinus(None=MaterialFunctionCall!(UVs, Rotation=@n5) {materialFunction=/Game/UI/Foundation/Materials/MF_UI_AngleGradient_DEPRECATED}.Result)
  n7 = LinearInterpolate(
   B=ScalarParameter() {defaultValue=0.98, parameterName=StartOffset}
   Alpha=ScalarParameter() {defaultValue=1, parameterName=PercentFill, sliderMax=1}
  ) {constA=0.99}
  n8 = ScalarParameter() {defaultValue=1, parameterName=PercentFill, sliderMax=1}
  n9 = LinearInterpolate(B=ScalarParameter() {defaultValue=0.02, parameterName=EdgeSoftness}, Alpha=@n8) {constA=0.01}
  n10 = Add(A=@n7, B=@n9)
  n11 = MaterialFunctionCall(Rotation=@n5) {materialFunction=&MF_UI_AngleGradient_DEPRECATED}
  n12 = LinearInterpolate(B=ScalarParameter() {defaultValue=0.9, parameterName=Terminus}, Alpha=@n8) {constA=1, constB=0}
  n13 = Add(A=@n12, B=@n9)
  n14 = Max(
   A=Multiply(A=MaterialFunctionCall(Alpha=@n6, Min=@n7, Max=@n10) {materialFunction=&SmoothStep}.Result)
   B=Multiply(A=MaterialFunctionCall(Alpha=@n11.Result, Min=@n12, Max=@n13) {materialFunction=&SmoothStep}.Result)
  )
  n15 = Panner(
   Coordinate=Multiply(A=TextureCoordinate(), B=ScalarParameter() {parameterName=NoiseScale})
   Speed=VectorParameter() {parameterName=NoiseSpeed, defaultValue=<rgba 0 0 0 1>}.RGB
  )
  LinearInterpolate(
   A=VectorParameter() {parameterName=GlowColorOuter, defaultValue=<rgba 0.2136 0 1 1>}.RGB
   B=VectorParameter() {parameterName=GlowColor, defaultValue=<rgba 0 1 1 1>}.RGB
   Alpha=Multiply(
    A=Multiply(A=@n4, B=@n14)
    B=TextureSample(UVs=@n15) {samplerType=SAMPLERTYPE_Grayscale, texture=&LowResBlurredNoise}.RGB
   )
  )
 ]
}
