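# Translucent UI material that draws a vehicle HUD horizon indicator procedurally, driven by Rotation and Pitch parameters.
# Rotated UVs and a pitch offset feed paired SmoothSteps that cut a line, with LineWidth scaled by UI pixel size.
# A central radial mask trims the line, and another radial gradient both bounds it and forms an outer ring.
# The ring is clipped by a center-scaled UV mask, then merged with the line by Max.
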Material /Game/Athena/HUD/Vehicles/Materials/M_UI_VehicleHorizon
graph {
 properties {blendMode=BLEND_Translucent, materialDomain=MD_UI, pixelDepthOffsetMode=PDOM_Legacy, usedShadingModels=Default Lit}
 nodes [
  n1 = ScalarParameter() {parameterName=Rotation, sliderMax=1, sliderMin=-1}
  n2 = ScalarParameter() {defaultValue=1, parameterName=RotationClamp}
  Clamp(None=Multiply(A=@n1, B=ScalarParameter() {defaultValue=1, parameterName=RotationDampening}), Min=Multiply(A=@n2) {constB=-1}, Max=@n2)
  Constant() {r=1}
  n3 = MaterialFunctionCall(UVs=TextureCoordinate(), Radius=ScalarParameter() {defaultValue=0.5, parameterName=OuterRadius}) {materialFunction=&RadialGradientExponential}
  n4 = ScalarParameter() {parameterName=OuterRing}
  n5 = Add(A=@n4, B=ScalarParameter() {defaultValue=0.01, parameterName=RingEdgeSoftness})
  n6 = MaterialFunctionCall(Alpha=@n3.RadialGradientExponential, Min=@n4, Max=@n5) {materialFunction=&SmoothStep}
  n7 = VectorParameter() {parameterName=RotationConstraints, defaultValue=<rgba 0 1 0 1>}
  n8 = Divide(A=Add(A=@n1))
  n9 = ComponentMask(None=MaterialFunctionCall(UVs=TextureCoordinate(), "Rotation Angle (0-1)"=LinearInterpolate(A=@n7.R, B=@n7.R, Alpha=@n8)) {materialFunction=&CustomRotator}.Rotated Values) {g=true}
  n10 = VectorParameter() {parameterName=PitchConstraints, defaultValue=<rgba 0 1 0 1>}
  n11 = ScalarParameter() {defaultValue=0.5, parameterName=Pitch, sliderMax=1, sliderMin=-1}
  n12 = LinearInterpolate(A=@n10.R, B=@n10.R, Alpha=Divide(A=Add(A=@n11)))
  n13 = Add(A=@n12, B=ScalarParameter() {defaultValue=1, parameterName=EdgeSoftness})
  n14 = MaterialFunctionCall() {materialFunction=&GetUserInterfaceUV}
  n15 = Add(
   A=@n13
   B=Divide(
    A=ScalarParameter() {defaultValue=0.1, parameterName=LineWidth}
    B=MaterialFunctionCall(Float2=@n14.Pixel Size) {materialFunction=&BreakOutFloat2Components}.G
   )
  )
  n16 = ScalarParameter() {defaultValue=1, parameterName=EdgeSoftness}
  n17 = Subtract(
   A=MaterialFunctionCall(Alpha=@n9, Min=@n12, Max=@n13) {materialFunction=&SmoothStep}.Result
   B=MaterialFunctionCall(Alpha=@n9, Min=@n15, Max=Add(A=@n15, B=@n16)) {materialFunction=&SmoothStep}.Result
  )
  n18 = ScalarParameter() {defaultValue=0.5, parameterName=MaskSize}
  n19 = MaterialFunctionCall(
   Alpha=MaterialFunctionCall() {materialFunction=&RadialGradientExponential}.RadialGradientExponential
   Min=@n18
   Max=Add(A=@n18, B=ScalarParameter() {defaultValue=1, parameterName=EdgeSoftness})
  ) {materialFunction=&SmoothStep}
  n20 = TextureCoordinate()
  n21 = VectorParameter() {parameterName=OuterRingMask, defaultValue=<rgba 1 0.5 0 1>}
  n22 = AppendVector(A=@n21.R, B=@n21.R)
  n23 = Add(A=@n5, B=ScalarParameter() {defaultValue=0.1, parameterName=OuterRingWidth})
  n24 = MaterialFunctionCall(Alpha=@n3.RadialGradientExponential, Min=@n23, Max=Add(A=@n23, B=ScalarParameter() {defaultValue=2, parameterName=RingEdgeInnerSoftness})) {materialFunction=&SmoothStep}
  Max(
   A=Multiply(A=OneMinus(None=@n6.Result), B=Subtract(A=@n17, B=@n19.Result))
   B=Multiply(
    A=MaterialFunctionCall(UVs=@n20, Texture Scale=@n22) {materialFunction=&ScaleUVsByCenter}.0-1 mask
    B=Subtract(A=@n6.Result, B=@n24.Result)
   )
  )
 ]
}
