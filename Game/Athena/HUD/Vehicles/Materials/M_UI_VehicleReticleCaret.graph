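# Additive UI material that procedurally draws a caret shape, named for the vehicle reticle in the HUD.
# UVs rotate by CaretRotation, then feed two center-scaled square masks whose sizes come from CaretPosition and CaretWidth.
# A soft vertical step set by BottomEdge and LineWidth is maxed into the outer square before the inner square is subtracted.
# A second step trims everything past BottomEdge, leaving a caret joined to a horizontal band, without textures.
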Material /Game/Athena/HUD/Vehicles/Materials/M_UI_VehicleReticleCaret
graph {
 properties {blendMode=BLEND_Additive, materialDomain=MD_UI, pixelDepthOffsetMode=PDOM_Legacy, usedShadingModels=Default Lit}
 nodes [
  Constant() {r=1}
  n1 = MaterialFunctionCall(UVs=TextureCoordinate(), "Rotation Angle (0-1)"=ScalarParameter() {defaultValue=0.125, parameterName=CaretRotation}) {materialFunction=&CustomRotator}
  n2 = ScalarParameter() {defaultValue=1, parameterName=CaretPosition}
  n3 = MaterialFunctionCall(UVs=@n1.Rotated Values, Texture Scale=@n2) {materialFunction=&ScaleUVsByCenter}
  n4 = ComponentMask(None=TextureCoordinate()) {g=true}
  n5 = ScalarParameter() {defaultValue=0.5, parameterName=BottomEdge}
  n6 = Subtract(A=@n5, B=ScalarParameter() {defaultValue=0.3286, parameterName=LineWidth})
  n7 = ScalarParameter() {defaultValue=0.001, parameterName=EdgeSoftness}
  n8 = MaterialFunctionCall(Alpha=@n4, Min=@n6, Max=Add(A=@n6, B=@n7)) {materialFunction=&SmoothStep}
  n9 = Subtract(A=@n2, B=ScalarParameter() {defaultValue=0.1, parameterName=CaretWidth})
  n10 = Add(A=@n5, B=@n7)
  Multiply(
   A=Subtract(
    A=Max(A=@n3.0-1 mask, B=@n8.Result)
    B=MaterialFunctionCall(UVs=@n1.Rotated Values, Texture Scale=@n9) {materialFunction=&ScaleUVsByCenter}.0-1 mask
   )
   B=OneMinus(None=MaterialFunctionCall(Alpha=@n4, Min=@n5, Max=@n10) {materialFunction=&SmoothStep}.Result)
  )
 ]
}
