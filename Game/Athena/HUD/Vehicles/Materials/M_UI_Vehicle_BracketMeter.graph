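# Translucent UI material for a vehicle HUD bracket meter, apparently for fuel, drawn mostly from procedural UV rectangle masks.
# An anchored, resizable bar combines a pixel-width border with a rotatable, screen-aligned segment pattern made by smoothstepping a gradient texture.
# An outline with an inner cutaway forms the bracket, joined by a movable caret that blends between arrow and fuel icon.
# Border, bar and optional low-fuel band tint the color, while the combined masks drive opacity.
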Material /Game/Athena/HUD/Vehicles/Materials/M_UI_Vehicle_BracketMeter
graph {
 properties {blendMode=BLEND_Translucent, materialDomain=MD_UI, pixelDepthOffsetMode=PDOM_Legacy, usedShadingModels=Default Lit}
 nodes [
  n1 = AppendVector(
   A=ScalarParameter() {defaultValue=0.2, group=Bar, parameterName=BarWidth}
   B=ScalarParameter() {defaultValue=0.5, group=Bar, parameterName=BarHeight}
  )
  n2 = VectorParameter() {group=Bar, parameterName=BarCenter, defaultValue=<rgba 0 1 1 1>}
  n3 = MaterialFunctionCall!(Texture Scale=@n1, UVs, Offset XY Texturesize Z=@n2.RGB) {materialFunction=/Engine/Functions/Engine_MaterialFunctions01/Texturing/ScaleUVsAroundPoint}
  n4 = MaterialFunctionCall() {materialFunction=&GetUserInterfaceUV}
  n5 = Subtract(A=@n1, B=Divide(A=ScalarParameter() {defaultValue=2, group=Bar, parameterName=BarBorderWidth}, B=@n4.Pixel Size))
  n6 = ScalarParameter() {defaultValue=32, group=Pattern, parameterName=PatternSize}
  n7 = MaterialFunctionCall(
   UVs=MaterialFunctionCall(Texture Dimensions=Divide(A=@n6, B=@n4.Pixel Size)) {materialFunction=&ScreenAlignedUVs}."X 100%, Y 100%"
   "Rotation Angle (0-1)"=ScalarParameter() {group=Pattern, parameterName=PatternRotation}
  ) {materialFunction=&CustomRotator}
  n8 = AppendVector(A=Constant() {r=1}, B=ScalarParameter() {defaultValue=50, group=Pattern, parameterName=PatternRepeat})
  n9 = ScalarParameter() {defaultValue=0.9, group=Pattern, parameterName=PatternEdge, sliderMax=1}
  n10 = MaterialFunctionCall(
   Alpha=TextureSample(UVs=Multiply(A=@n7.Rotated Values, B=@n8)) {texture=/Game/Global/Textures/Default/Masks_and_Gradients/T_UI_ReticleSegmentGradient}.R
   Min=@n9
   Max=Add(A=@n9, B=ScalarParameter() {group=Pattern, parameterName=PatternEdgeSmooth})
  ) {materialFunction=&SmoothStep}
  n11 = Max(
   A=Subtract(
    A=@n3.0-1 mask
    B=MaterialFunctionCall(Texture Scale=@n5, Offset XY Texturesize Z=@n2.RGB) {materialFunction=&ScaleUVsAroundPoint}.0-1 mask
   )
   B=Multiply(A=@n3.0-1 mask, B=OneMinus(None=@n10.Result))
  )
  n12 = ScalarParameter() {defaultValue=4, group=Border, parameterName=OutlineSize}
  n13 = MaterialFunctionCall() {materialFunction=&GetUserInterfaceUV}
  n14 = MaterialFunctionCall(Texture Scale=OneMinus(None=Divide(A=@n12, B=@n13.Pixel Size))) {materialFunction=&ScaleUVsByCenter}
  n15 = AppendVector(A=ScalarParameter() {defaultValue=0.7, group=Border, parameterName=InnerCutaway}, B=Constant() {r=1})
  n16 = VectorParameter() {group=Border, parameterName=CutawayCenter, defaultValue=<rgba 1 0 1 1>}
  n17 = ScalarParameter() {group=Caret, parameterName=CaretX}
  n18 = LinearInterpolate(Alpha=ScalarParameter() {group=Caret, parameterName=CaretY, sliderMax=1}) {constA=-0.5, constB=0.5}
  n19 = MaterialFunctionCall(
   UVs=Add(A=TextureCoordinate() {uTiling=-1}, B=AppendVector(A=@n17, B=@n18))
   Texture Scale=Divide(
    A=ScalarParameter() {defaultValue=32, group=Caret, parameterName=CaretSize}
    B=MaterialFunctionCall() {materialFunction=&GetUserInterfaceUV}.Pixel Size
   )
  ) {materialFunction=&ScaleUVsByCenter}
  n20 = TextureSample(UVs=@n19.UVs) {samplerSource=SSM_Clamp_WorldGroupSettings, texture=/Game/Athena/HUD/Materials/T_UI_LowFuelIcon}
  n21 = ScalarParameter() {defaultValue=0.15, parameterName=Edge}
  n22 = Add(A=@n21, B=ScalarParameter() {defaultValue=0.05, parameterName=EdgeSoftness})
  n23 = Max(
   A=Subtract(
    A=OneMinus(None=@n14.0-1 mask)
    B=MaterialFunctionCall(Texture Scale=@n15, Offset XY Texturesize Z=@n16.RGB) {materialFunction=&ScaleUVsAroundPoint}.0-1 mask
   )
   B=LinearInterpolate(
    A=TextureSample(UVs=@n19.UVs) {texture=/Game/UI/Foundation/Textures/Buttons/T_CycleArrow}.A
    B=MaterialFunctionCall(Alpha=@n20.R, Min=@n21, Max=@n22) {materialFunction=&SmoothStep}.Result
    Alpha=ScalarParameter() {parameterName=UseFuelIcon}
   )
  )
  n24 = LinearInterpolate(
   A=VectorParameter() {parameterName=BorderColor, defaultValue=<rgba 1 1 1 0.9>}.RGB
   B=VectorParameter() {parameterName=BarColor, defaultValue=<rgba 1 1 1 1>}.RGB
   Alpha=Subtract(A=@n11, B=@n23)
  )
  n25 = MaterialFunctionCall(
   Texture Scale=AppendVector(
    A=ScalarParameter() {defaultValue=0.2, group=Bar, parameterName=BarWidth}
    B=ScalarParameter() {defaultValue=0.2, group=Bar, parameterName=LowFuelHeight}
   )
   Offset XY Texturesize Z=VectorParameter() {group=Bar, parameterName=BarCenter, defaultValue=<rgba 0 1 1 1>}.RGB
  ) {materialFunction=&ScaleUVsAroundPoint}
  LinearInterpolate(
   A=@n24
   B=LinearInterpolate(
    A=@n24
    B=VectorParameter() {parameterName=LowFuelColor, defaultValue=<rgba 1 0.117 0.0006 1>}.RGB
    Alpha=Subtract(A=@n25.0-1 mask, B=@n23) {constB=0}
   )
   Alpha=ScalarParameter() {parameterName=HasLowFuelIndicator}
  ) {constAlpha=0}
  Max(A=@n23, B=Multiply(A=@n11, B=VectorParameter() {parameterName=BarColor, defaultValue=<rgba 1 1 1 1>}.A))
  ScalarParameter() {defaultValue=0.5, group=Pattern, parameterName=PatternOpacity}
 ]
}
