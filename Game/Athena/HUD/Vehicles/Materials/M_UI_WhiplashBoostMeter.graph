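# Translucent UI material for the Whiplash vehicle's boost meter, drawn as a partial-arc gauge.
# Two smoothstepped radial gradients, one slightly off-center, subtract to form the ring mask.
# A rotated angle gradient, using a deprecated function, clips the arc at Terminus and builds the PercentFill mask.
# That mask blends background into a two-color linear-gradient fill, with a scalar-driven color override, for both color and alpha.
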Material /Game/Athena/HUD/Vehicles/Materials/M_UI_WhiplashBoostMeter
graph {
 properties {blendMode=BLEND_Translucent, materialDomain=MD_UI, pixelDepthOffsetMode=PDOM_Legacy, usedShadingModels=Default Lit}
 nodes [
  n1 = ScalarParameter() {defaultValue=0.02, parameterName=Edge}
  n2 = Add(A=@n1, B=Multiply(A=@n1, B=ScalarParameter() {parameterName=EdgeSoftness}) {constB=1.01})
  n3 = MaterialFunctionCall(
   Alpha=MaterialFunctionCall() {materialFunction=&RadialGradientExponential}.RadialGradientExponential
   Min=@n1
   Max=@n2
  ) {materialFunction=&SmoothStep}
  n4 = Add(A=@n2, B=ScalarParameter() {defaultValue=0.01, parameterName=EdgeWidth})
  n5 = ScalarParameter() {defaultValue=0.2, parameterName=EdgeSoftness}
  n6 = MaterialFunctionCall(
   Alpha=MaterialFunctionCall(CenterPosition=Constant2Vector() {g=0.525, r=0.525}) {materialFunction=&RadialGradientExponential}.RadialGradientExponential
   Min=@n4
   Max=Add(A=@n4, B=Multiply(A=@n4, B=@n5) {constB=1.01})
  ) {materialFunction=&SmoothStep}
  n7 = VectorParameter() {parameterName=BGColor, defaultValue=<rgba 0 0 0 1>}
  n8 = VectorParameter() {parameterName=FillColor, defaultValue=<rgba 0 0.604 1 1>}
  n9 = VectorParameter() {parameterName=FillColorOverride, defaultValue=<rgba 1 1 1 1>}
  n10 = ScalarParameter() {parameterName=ColorOverrideScalar}
  n11 = LinearInterpolate(A=@n8.A, B=@n9.A, Alpha=@n10)
  n12 = MaterialFunctionCall!(UVs, Rotation=ScalarParameter() {defaultValue=-0.62, parameterName=AngleRotation}) {materialFunction=/Game/UI/Foundation/Materials/MF_UI_AngleGradient_DEPRECATED}
  n13 = ScalarParameter() {defaultValue=0.5, parameterName=PercentFill, sliderMax=1}
  n14 = LinearInterpolate(A=OneMinus(None=ScalarParameter() {defaultValue=0.24, parameterName=Terminus}), Alpha=OneMinus(None=Saturate(None=@n13))) {constB=0}
  n15 = Multiply(A=ScalarParameter() {parameterName=PercentEdgeSoftness}, B=@n14)
  n16 = OneMinus(None=MaterialFunctionCall(Alpha=OneMinus(None=@n12.Result), Min=@n14, Max=Add(A=@n14, B=@n15)) {materialFunction=&SmoothStep}.Result) {desc=Fill mask}
  n17 = Constant()
  Multiply(
   A=Multiply(A=Subtract(A=@n3.Result, B=@n6.Result), B=LinearInterpolate(A=@n7.A, B=@n11, Alpha=@n16))
   B=If(A=@n12.Result, B=ScalarParameter() {defaultValue=0.24, parameterName=Terminus}, "A > B"=Constant() {r=1}, "A == B"=@n17, "A < B"=@n17)
  )
  n18 = LinearInterpolate(
   A=@n8.RGB
   B=VectorParameter() {parameterName=FillColor2, defaultValue=<rgba 0.1518 1 0 1>}.RGB
   Alpha=MaterialFunctionCall(
    GradientStart=ScalarParameter() {parameterName=Start}
    GradientEnd=ScalarParameter() {parameterName=End}
    RotationAngle=ScalarParameter() {parameterName=Direction}
   ) {materialFunction=&MF_UI_LinearGradient}.Result
  )
  Reroute(None=LinearInterpolate(A=@n7.RGB, B=LinearInterpolate(A=@n18, B=@n9.RGB, Alpha=@n10), Alpha=@n16))
 ]
}
