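# Sound cue for the Ostrich vehicle, named for an ability being disabled.
# A single wave player plays the Veh_Ostrich_Ability_Disabled_01 sound with no randomization or modulation.
# It uses the Ostrich close-movement attenuation and the weapon foley sound class.
# Overridden concurrency allows one instance at a time, preventing new plays while one is active, with a retrigger time set.
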SoundCue /Game/Athena/Sounds/Vehicles/Ostrich/Veh_Ostrich_Ability_Disabled_Cue
root=WavePlayer_0
nodes [{id=WavePlayer_0, properties {soundWaveAssetPtr=&/Veh_Ostrich_Ability_Disabled_01}}]
properties {
 attenuationSettings=/Game/Sounds/Attenuation/Fort_Ostrich_Movement_Close_Att
 bOverrideConcurrency=true
 maxDistance=6512
 soundClassObject=&Fort_WeaponFoley
 volumeMultiplier=1
 concurrencyOverrides {maxCount=1, resolutionRule=PreventNew, retriggerTime=1}
}
