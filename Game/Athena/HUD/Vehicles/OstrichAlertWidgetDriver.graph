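# HUD alert widget for the Ostrich vehicle's driver that warns when the jump, stomp or dash ability is blocked.
# Each blocked event sets its message on a slanted warning banner, shows it, plays a disabled sound and starts a two-second timer to collapse it.
# Construct clears the text and hides the banner, while underwater-attempt events only print debug strings.
# A fade animation named Play is defined but never triggered in this graph.
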WidgetBlueprint /Game/Athena/HUD/Vehicles/OstrichAlertWidgetDriver
animations [Play {
  binding_count=3
  length_seconds=2
  object=&self:Play
  sequence {bindings [
    Image_5 {
     id=#1
     possessed_class=Image
     tracks [{
       type=MovieSceneFloatTrack
       display_name=Render Opacity
       sections [{type=MovieSceneFloatSection, channels [None {type=MovieSceneScriptingFloatChannel, keys [[0,0],[6,0.25]]}]}]
      }]
    }
    Image_3 {
     id=#2
     possessed_class=Image
     tracks [{
       type=MovieSceneFloatTrack
       display_name=Render Opacity
       sections [{type=MovieSceneFloatSection, channels [None {type=MovieSceneScriptingFloatChannel, keys [[0,0],[4,1]]}]}]
      }]
    }
    Overlay_Content {
     id=#3
     possessed_class=Overlay
     tracks [
      {
       type=MovieScene2DTransformTrack
       sections [{
         type=MovieScene2DTransformSection
         channels [
          Translation.X {type=MovieSceneScriptingFloatChannel}
          Translation.Y {type=MovieSceneScriptingFloatChannel}
          Angle {type=MovieSceneScriptingFloatChannel}
          Scale.X {type=MovieSceneScriptingFloatChannel}
          Scale.Y {type=MovieSceneScriptingFloatChannel}
          Shear.X {type=MovieSceneScriptingFloatChannel}
          Shear.Y {type=MovieSceneScriptingFloatChannel}
         ]
        }]
      }
      {
       type=MovieSceneFloatTrack
       display_name=Render Opacity
       sections [{type=MovieSceneFloatSection, channels [None {type=MovieSceneScriptingFloatChannel, keys [[0,0],[4,1],[35,1],[40,0]]}]}]
      }
     ]
    }
   ]}
 }]
graphs [EventGraph {nodes [
   n1 = HandleHideMessage() {is=title K2Node_CustomEvent}
   n2 = Get Overlay_Content() -> (Overlay_Content:Overlay Object Reference)
   > SetVisibility(self=@n2, InVisibility=Collapsed)
   Event Construct()
   > n3 = Set MessageText(MessageText:Text=GetEmptyText!() -> (ReturnValue:Text)) -> (Output_Get:Text)
   > "SetText (Text)"(self=Get CommonTextBlock_WarningText() -> (CommonTextBlock_WarningText:Common Text Object Reference), InText=@n3)
   > SetVisibility(self=Get Overlay_Content(), InVisibility=Collapsed)
   Event OnDriverJumpAbilityBlocked() {is=title K2Node_Event}
   > n4 = Set MessageText(MessageText="LOCTEXT(\"JUMP DISABLED\")")
   "SetText (Text)"(execute=[@n4.then, @n5.then, @n6.then], self=Get CommonTextBlock_WarningText(), InText=Get MessageText() -> (MessageText:Text))
   > Set Timer by Event(Delegate=@n1, Time=2, bLooping=false)
   > SetVisibility(self=@n2, InVisibility=HitTestInvisible)
   > PlaySound2D(Sound=/Game/Athena/Sounds/Vehicles/Ostrich/Veh_Ostrich_Ability_Disabled_Cue)
   Event OnAttemptJumpUnderwater() {is=title K2Node_Event}
   > PrintString!(execute, InString="OstrichAlert - Attempted to Jump under water!", bPrintToScreen, bPrintToLog, TextColor, Duration, Key) -> (then)
   Event OnAttemptStompUnderwater() {is=title K2Node_Event}
   > PrintString(InString="OstrichAlert - Attempted to Stomp under water!")
   Event OnAttemptDashUnderwater() {is=title K2Node_Event}
   > PrintString(InString="OstrichAlert - Attempted to Dash under water!")
   Event OnDriverStompAbilityBlocked() {is=title K2Node_Event}
   > n5 = Set MessageText(MessageText="LOCTEXT(\"STOMP DISABLED\")")
   Event OnDriverDashAbilityBlocked() {is=title K2Node_Event}
   > n6 = Set MessageText(MessageText="LOCTEXT(\"DASH DISABLED\")")
   Event PreConstruct()
   Event Tick()
  ]}]
variables [MessageText {type=string, default=""}]
widgets {
 info {inheritedWidgetCount=1, parentClass=/Script/FortniteUI.OstrichAlertWidgetDriver, rootWidgetClass=&Overlay, widgetCount=8}
 widgets [
  Overlay_Content {widgetClassPath=&Overlay, bIsVariable=true, widget=Overlay_Content, properties {visibility=HitTestInvisible}}
  WarningMessaging {
   widgetClassPath=&Overlay
   parent=Overlay_Content
   widget=WarningMessaging
   properties {renderTransform {angle=0, scale=<1 1>, shear=<-16 0>, translation=<0 336>}}
   slot {class=OverlaySlot, properties {horizontalAlignment=HAlign_Center, verticalAlignment=VAlign_Center}}
  }
  AdditonalOuterRing {
   widgetClassPath=&Image
   bIsVariable=true
   parent=WarningMessaging
   widget=AdditonalOuterRing
   properties {brush {
     bIsDynamicallyLoaded=false
     drawAs=Box
     imageType=NoImage
     mirroring=NoMirror
     resourceName=None
     resourceObject=/Game/Athena/HUD/Vehicles/Materials/M_UI_OstrichWarningText_Outline
     tiling=NoTile
     imageSize=<24 24>
     margin=<ltrb 0.5 0.5 0.5 0.5>
     outlineSettings {
      bUseBrushTransparency=false
      roundingType=HalfHeightRadius
      width=0
      color {colorUseRule=UseColor_Specified, specifiedColor=<rgba 0 0 0 0>}
      cornerRadii=<0 0 0 0>
     }
     tintColor {colorUseRule=UseColor_Specified, specifiedColor=<rgba 1 1 1 1>}
     uVRegion {bIsValid=false, max=<0 0>, min=<0 0>}
    }}
   slot {class=OverlaySlot, properties {horizontalAlignment=HAlign_Fill, verticalAlignment=VAlign_Fill, padding=<ltrb -8 -8 -8 -8>}}
  }
  Image_3 {
   widgetClassPath=&Image
   bIsVariable=true
   parent=WarningMessaging
   widget=Image_3
   properties {
    brush {
     bIsDynamicallyLoaded=false
     drawAs=Box
     imageType=NoImage
     mirroring=NoMirror
     resourceName=None
     resourceObject=/Game/Athena/HUD/Vehicles/Materials/M_UI_OstrichWarningText_Backing
     tiling=NoTile
     imageSize=<16 16>
     margin=<ltrb 0.5 0.5 0.5 0.5>
     outlineSettings {
      bUseBrushTransparency=false
      roundingType=HalfHeightRadius
      width=0
      color {colorUseRule=UseColor_Specified, specifiedColor=<rgba 0 0 0 0>}
      cornerRadii=<0 0 0 0>
     }
     tintColor {colorUseRule=UseColor_Specified, specifiedColor=<rgba 1 1 1 1>}
     uVRegion {bIsValid=false, max=<0 0>, min=<0 0>}
    }
    renderTransformPivot=<1 0.5>
   }
   slot {class=OverlaySlot, properties {horizontalAlignment=HAlign_Fill, verticalAlignment=VAlign_Fill}}
  }
  Image_4 {
   widgetClassPath=&Image
   bIsVariable=true
   parent=WarningMessaging
   widget=Image_4
   properties {
    renderOpacity=0.5
    brush {
     bIsDynamicallyLoaded=false
     drawAs=Box
     imageType=NoImage
     mirroring=NoMirror
     resourceName=None
     resourceObject=&M_UI_OstrichWarningText_Outline
     tiling=NoTile
     imageSize=<16 16>
     margin=<ltrb 0.5 0.5 0.5 0.5>
     outlineSettings {
      bUseBrushTransparency=false
      roundingType=HalfHeightRadius
      width=0
      color {colorUseRule=UseColor_Specified, specifiedColor=<rgba 0 0 0 0>}
      cornerRadii=<0 0 0 0>
     }
     tintColor {colorUseRule=UseColor_Specified, specifiedColor=<rgba 1 1 1 1>}
     uVRegion {bIsValid=false, max=<0 0>, min=<0 0>}
    }
   }
   slot {class=OverlaySlot, properties {horizontalAlignment=HAlign_Fill, verticalAlignment=VAlign_Fill}}
  }
  Image_5 {
   widgetClassPath=&Image
   bIsVariable=true
   parent=WarningMessaging
   widget=Image_5
   properties {
    renderOpacity=0.25
    visibility=SelfHitTestInvisible
    brush {
     bIsDynamicallyLoaded=false
     drawAs=Box
     imageType=NoImage
     mirroring=NoMirror
     resourceName=None
     resourceObject=/Game/Athena/HUD/LTM/Wax/Materials/M_UI_TextGlow_Add_Gradient_Orange_Nevada
     tiling=NoTile
     imageSize=<64 64>
     margin=<ltrb 0.5 0.5 0.5 0.5>
     outlineSettings {
      bUseBrushTransparency=false
      roundingType=HalfHeightRadius
      width=0
      color {colorUseRule=UseColor_Specified, specifiedColor=<rgba 0 0 0 0>}
      cornerRadii=<0 0 0 0>
     }
     tintColor {colorUseRule=UseColor_Specified, specifiedColor=<rgba 1 1 1 1>}
     uVRegion {bIsValid=false, max=<0 0>, min=<0 0>}
    }
   }
   slot {class=OverlaySlot, properties {horizontalAlignment=HAlign_Fill, verticalAlignment=VAlign_Fill, padding=<ltrb -24 -24 -24 -24>}}
  }
  CommonTextBlock_WarningText {
   widgetClassPath=&CommonTextBlock
   bIsVariable=true
   parent=WarningMessaging
   widget=CommonTextBlock_WarningText
   properties {
    justification=Center
    text="[MESSAGE]"
    textTransformPolicy=ToUpper
    colorAndOpacity {colorUseRule=UseColor_Specified, specifiedColor=<rgba 0 0 0 0.5>}
    font {
     bForceMonospaced=false
     bMaterialIsStencil=false
     fontMaterial=null
     fontObject=&BurbankSmall
     letterSpacing=360
     monospacedWidth=1
     size=15
     skewAmount=0
     typefaceFontName=Black
     outlineSettings {
      bApplyOutlineToDropShadows=false
      bMiteredCorners=false
      bSeparateFillAlpha=true
      outlineMaterial=null
      outlineSize=2
      outlineColor=<rgba 1 1 1 0.125>
     }
    }
    renderTransform {angle=0, scale=<1 1>, shear=<16 0>, translation=<0 0>}
    shadowOffset=<0 0>
   }
   slot {class=OverlaySlot, properties {horizontalAlignment=HAlign_Center, verticalAlignment=VAlign_Center, padding=<ltrb 12 6 12 6>}}
  }
  TutorialHighlight {widgetClassPath=&TutorialHighlight}
 ]
}
