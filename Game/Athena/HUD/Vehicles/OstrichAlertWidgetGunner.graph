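# Gunner HUD alert widget for the Ostrich vehicle that briefly warns when missiles or shotgun are disabled.
# Blocked-weapon events from the native FortniteUI parent set MessageText, then share a path that updates the centered warning text, shows the overlay and plays a disabled sound.
# A one-shot timer collapses the overlay, which also starts collapsed with empty text.
# Underwater fire attempts only print debug strings.
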WidgetBlueprint /Game/Athena/HUD/Vehicles/OstrichAlertWidgetGunner
graphs [EventGraph {nodes [
   n1 = HandleHideMessage() {is=title K2Node_CustomEvent}
   n2 = Get Overlay_Content() -> (Overlay_Content:Overlay Object Reference)
   > SetVisibility(self=@n2, InVisibility=Collapsed)
   Event Construct()
   > n3 = Set MessageText(MessageText:Text=GetEmptyText!() -> (ReturnValue:Text)) -> (Output_Get:Text)
   > "SetText (Text)"(self=Get CommonTextBlock_WarningText() -> (CommonTextBlock_WarningText:Common Text Object Reference), InText=@n3)
   > SetVisibility(self=Get Overlay_Content(), InVisibility=Collapsed)
   Event OnAttemptFireMissileUnderwater() {is=title K2Node_Event}
   > PrintString!(execute, InString="OstrichAlert - Attempted to fire Missile under water!", bPrintToScreen, bPrintToLog, TextColor, Duration, Key) -> (then)
   Event OnAttemptFireShotgunUnderwater() {is=title K2Node_Event}
   > PrintString(InString="OstrichAlert - Attempted to fire Shotgun under water!")
   Event OnGunnerMissilesBlocked() {is=title K2Node_Event}
   > n4 = Set MessageText(MessageText="LOCTEXT(\"MISSLES DISABLED\")")
   "SetText (Text)"(execute=[@n4.then, @n5.then], self=Get CommonTextBlock_WarningText(), InText=Get MessageText() -> (MessageText:Text))
   > Set Timer by Event(Delegate=@n1, Time=2, bLooping=false)
   > SetVisibility(self=@n2, InVisibility=HitTestInvisible)
   > PlaySound2D(Sound=/Game/Athena/Sounds/Vehicles/Ostrich/Veh_Ostrich_Ability_Disabled_Cue)
   Event OnGunnerShotgunBlocked() {is=title K2Node_Event}
   > n5 = Set MessageText(MessageText="LOCTEXT(\"SHOTGUN  DISABLED\")")
   Event PreConstruct()
   Event Tick()
  ]}]
variables [MessageText {type=string, default=""}]
widgets {
 info {inheritedWidgetCount=1, parentClass=/Script/FortniteUI.OstrichAlertWidgetGunner, rootWidgetClass=&Overlay, widgetCount=3}
 widgets [
  Overlay_Content {widgetClassPath=&Overlay, bIsVariable=true, widget=Overlay_Content, properties {visibility=HitTestInvisible}}
  CommonTextBlock_WarningText {
   widgetClassPath=&CommonTextBlock
   bIsVariable=true
   parent=Overlay_Content
   widget=CommonTextBlock_WarningText
   properties {
    justification=Center
    text="[MESSAGE]"
    textTransformPolicy=ToUpper
    colorAndOpacity {colorUseRule=UseColor_Specified, specifiedColor=<rgba 1 0.6445 0 1>}
    font {
     bForceMonospaced=false
     bMaterialIsStencil=false
     fontMaterial=null
     fontObject=&Burbank
     letterSpacing=0
     monospacedWidth=1
     size=24
     skewAmount=0
     typefaceFontName=BurbankRegular-Black
     outlineSettings {
      bApplyOutlineToDropShadows=false
      bMiteredCorners=false
      bSeparateFillAlpha=false
      outlineMaterial=null
      outlineSize=4
      outlineColor=<rgba 0 0 0 1>
     }
    }
    renderTransform {angle=0, scale=<1 1>, shear=<-12 0>, translation=<400 0>}
    shadowOffset=<0 0>
   }
   slot {class=OverlaySlot, properties {horizontalAlignment=HAlign_Center, verticalAlignment=VAlign_Center}}
  }
  TutorialHighlight {widgetClassPath=&TutorialHighlight}
 ]
}
