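# Vehicle HUD widget, based on OstrichFinaleWidget, that shows the self-destruct warning banner and countdown.
# Priming shows the banner, plays a stripe-progress animation using Duration directly as playback speed, and debug-prints it; cancelling hides it.
# Start switches stripe and glow materials and reveals the timer; updates interpolate separate seconds and millisecond numeric blocks.
# Finish hides the timer and displays a good-bye message.
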WidgetBlueprint /Game/Athena/HUD/Vehicles/OstrichFinale
animations [Prime {
  binding_count=1
  length_seconds=0.9966
  object=&self:Prime
  sequence {bindings [Image_WarningStripes {
     id=#1
     possessed_class=Image
     tracks [{
       type=MovieSceneWidgetMaterialTrack
       sections [{
         type=MovieSceneParameterSection
         end_frame=19
         start_frame=0
         channels [Progress {type=MovieSceneScriptingFloatChannel, keys [[0,0],[19.9327,1]]}]
        }]
      }]
    }]}
 }]
graphs [EventGraph {nodes [
   Event Construct()
   > SetBrushFromMaterial(
    self=Get Image_WarningStripes() -> (Image_WarningStripes:Image Object Reference)
    Material=/Game/Athena/HUD/Materials/M_UI_StripedWarningMaterial_Inst_Prime
   )
   > SetBrushFromMaterial(self=Get Image_GlowPulse() -> (Image_GlowPulse:Image Object Reference), Material=/Game/Athena/HUD/LTM/Wax/Materials/M_UI_TextGlow_Inst)
   n1 = Event OnSelfDestructStart!() -> (OutputDelegate, then, WarningDuration:float) {is=title K2Node_Event}
   n2 = float * float(A=@n1.WarningDuration, B=1)
   > SetCurrentValue!(
    execute
    self:Common Numeric Text Block Object Reference=Get TimeRemainingText() -> (TimeRemainingText:Common Numeric Text Block Object Reference)
    NewValue:float="To Float (Integer)"(InInt=Truncate(A=@n2))
   ) -> (then)
   > SetVisibility(
    self=[
     Get Overlay_Timer() -> (Overlay_Timer:Overlay Object Reference)
     Get Overlay_TopContainer() -> (Overlay_TopContainer:Overlay Object Reference)
    ]
    InVisibility=HitTestInvisible
   )
   > SetBrushFromMaterial(self=Get Image_WarningStripes(), Material=/Game/Athena/HUD/Materials/M_UI_StripedWarningMaterial_Inst)
   > "SetText (Text)"(self=Get CommonTextBlock_WarningText() -> (CommonTextBlock_WarningText:Common Text Object Reference), InText="LOCTEXT(\"Self Destruct\")")
   > SetBrushFromMaterial(self=Get Image_GlowPulse(), Material=/Game/Athena/HUD/LTM/Wax/Materials/M_UI_TextGlow_Pulse)
   Event OnSelfDestructFinish() {is=title K2Node_Event}
   > SetVisibility(self=Get Overlay_Timer(), InVisibility=Hidden)
   > "SetText (Text)"(self=Get CommonTextBlock_WarningText(), InText="LOCTEXT(\"Good-bye :(\")")
   n3 = Event OnUpdateSelfDestruct!() -> (OutputDelegate, then, RemainingTime:float) {is=title K2Node_Event}
   > InterpolateToValue!(
    execute
    self:Common Numeric Text Block Object Reference=Get TimeRemainingText()
    TargetValue:float=@n3.RemainingTime
    MaximumInterpolationDuration:float=Get DestructDisplayUpdateFrequency() -> (DestructDisplayUpdateFrequency:float)
    MinimumChangeRate:float=1
    OutroOffset:float=0
   ) -> (then)
   n4 = Truncate(A=float * float(A=@n3.RemainingTime, B=1000))
   n5 = int * int(A:int=Truncate(A=@n3.RemainingTime), B:int=1000) -> (ReturnValue:int=0) {is=title K2Node_CommutativeAssociativeBinaryOperator}
   > InterpolateToValue(
    self=Get TimeRemainingText_remainder() -> (TimeRemainingText_remainder:Common Numeric Text Block Object Reference)
    TargetValue="To Float (Integer)"(InInt=int - int(A=@n4, B=@n5) {is=title K2Node_CallFunction})
    MaximumInterpolationDuration=1
    MinimumChangeRate=1
    OutroOffset=0
   )
   n6 = Event OnStartSelfDestructPrime!() -> (OutputDelegate, then, Duration:float) {is=title K2Node_Event}
   > SetVisibility(self=Get Overlay_TopContainer(), InVisibility=HitTestInvisible)
   > PlayAnimation(InAnimation=Get Prime() -> (Prime:Widget Animation Object Reference), PlaybackSpeed=@n6.Duration)
   > PrintString!(
    execute
    InString="To String (Float)"(InDouble=@n6.Duration)
    bPrintToScreen
    bPrintToLog
    TextColor=(R=1.4,G=0.4753,B=0.4057,A=1)
    Duration
    Key
   ) -> (then)
   Event OnCancelSelfDestructPrime() {is=title K2Node_Event}
   > SetVisibility(self=Get Overlay_TopContainer(), InVisibility=Hidden)
   Event PreConstruct()
   Event Tick()
   Event OnSetupComplete() {is=title K2Node_Event}
  ]}]
variables []
widgets {
 info {inheritedWidgetCount=0, parentClass=/Script/FortniteUI.OstrichFinaleWidget, rootWidgetClass=&Overlay, widgetCount=12}
 widgets [
  Overlay_TopContainer {widgetClassPath=&Overlay, bIsVariable=true, widget=Overlay_TopContainer, properties {visibility=Hidden}}
  VerticalBox_1 {
   widgetClassPath=&VerticalBox
   parent=Overlay_TopContainer
   widget=VerticalBox_1
   slot {class=OverlaySlot, properties {horizontalAlignment=HAlign_Center, padding=<ltrb 0 180 0 0>}}
  }
  Overlay_2 {
   widgetClassPath=&Overlay
   parent=VerticalBox_1
   widget=Overlay_2
   slot {class=VerticalBoxSlot, properties {horizontalAlignment=HAlign_Left, verticalAlignment=VAlign_Top}}
  }
  Image_GlowPulse {
   widgetClassPath=&Image
   bIsVariable=true
   parent=Overlay_2
   widget=Image_GlowPulse
   properties {
    brush {
     bIsDynamicallyLoaded=false
     drawAs=Image
     imageType=NoImage
     mirroring=NoMirror
     resourceName=None
     resourceObject=&M_UI_TextGlow_Pulse
     tiling=NoTile
     imageSize=<32 32>
     margin=<ltrb 0 0 0 0>
     outlineSettings {
      bUseBrushTransparency=false
      roundingType=HalfHeightRadius
      width=0
      color {colorUseRule=UseColor_Specified, specifiedColor=<rgba 0 0 0 0>}
      cornerRadii=<0 0 0 0>
     }
     tintColor {colorUseRule=UseColor_Specified, specifiedColor=<rgba 1 1 1 1>}
     uVRegion {bIsValid=false, max=<0 0>, min=<0 0>}
    }
    colorAndOpacity=<rgba 0.8469 0.1384 0.0168 1>
   }
   slot {class=OverlaySlot, properties {horizontalAlignment=HAlign_Fill, verticalAlignment=VAlign_Fill, padding=<ltrb -200 -100 -200 -100>}}
  }
  RetainerBox_AdditiveContainer {
   widgetClassPath=/Script/UMG.RetainerBox
   bIsVariable=true
   parent=Overlay_2
   widget=RetainerBox_AdditiveContainer
   properties {effectMaterial=/Game/Athena/HUD/Materials/M_UI_SelfDestructRetainer_Inst}
   slot {class=OverlaySlot}
  }
  Overlay_1 {widgetClassPath=&Overlay, parent=RetainerBox_AdditiveContainer, widget=Overlay_1, slot {class=PanelSlot}}
  Image_WarningStripes {
   widgetClassPath=&Image
   bIsVariable=true
   parent=Overlay_1
   widget=Image_WarningStripes
   properties {
    brush {
     bIsDynamicallyLoaded=false
     drawAs=Image
     imageType=NoImage
     mirroring=NoMirror
     resourceName=None
     resourceObject=&M_UI_StripedWarningMaterial_Inst
     tiling=NoTile
     imageSize=<32 32>
     margin=<ltrb 0 0 0 0>
     outlineSettings {
      bUseBrushTransparency=false
      roundingType=HalfHeightRadius
      width=0
      color {colorUseRule=UseColor_Specified, specifiedColor=<rgba 0 0 0 0>}
      cornerRadii=<0 0 0 0>
     }
     tintColor {colorUseRule=UseColor_Specified, specifiedColor=<rgba 1 1 1 1>}
     uVRegion {bIsValid=false, max=<0 0>, min=<0 0>}
    }
    renderTransform {angle=0, scale=<1 1>, shear=<-38 0>, translation=<0 0>}
   }
   slot {class=OverlaySlot, properties {horizontalAlignment=HAlign_Fill, verticalAlignment=VAlign_Fill}}
  }
  CommonTextBlock_WarningText {
   widgetClassPath=&CommonTextBlock
   bIsVariable=true
   parent=Overlay_1
   widget=CommonTextBlock_WarningText
   properties {
    justification=Center
    text=Self Destruct?
    textTransformPolicy=ToUpper
    colorAndOpacity {colorUseRule=UseColor_Specified, specifiedColor=<rgba 1 0.6445 0 1>}
    font {
     bForceMonospaced=false
     bMaterialIsStencil=false
     fontMaterial=null
     fontObject=&Burbank
     letterSpacing=0
     monospacedWidth=1
     size=36
     skewAmount=0
     typefaceFontName=BurbankRegular-Black
     outlineSettings {
      bApplyOutlineToDropShadows=false
      bMiteredCorners=false
      bSeparateFillAlpha=false
      outlineMaterial=null
      outlineSize=4
      outlineColor=<rgba 0 0 0 1>
     }
    }
    renderTransform {angle=0, scale=<1 1>, shear=<-12 0>, translation=<0 0>}
    shadowOffset=<0 0>
   }
   slot {class=OverlaySlot, properties {padding=<ltrb 200 15 200 15>}}
  }
  Overlay_Timer {
   widgetClassPath=&Overlay
   bIsVariable=true
   parent=VerticalBox_1
   widget=Overlay_Timer
   properties {visibility=Hidden, renderTransform {angle=0, scale=<1 1>, shear=<-12 0>, translation=<0 -18>}}
   slot {class=VerticalBoxSlot, properties {horizontalAlignment=HAlign_Center, verticalAlignment=VAlign_Top}}
  }
  SizeBox_0 {
   widgetClassPath=&SizeBox
   parent=Overlay_Timer
   widget=SizeBox_0
   properties {bOverride_WidthOverride=true, widthOverride=65.6212}
   slot {class=OverlaySlot, properties {horizontalAlignment=HAlign_Center}}
  }
  TimeRemainingText {
   widgetClassPath=/Script/CommonUI.CommonNumericTextBlock
   bIsVariable=true
   parent=SizeBox_0
   widget=TimeRemainingText
   properties {
    currentNumericValue=1
    interpolationUpdateInterval=0.1
    justification=Center
    text=Text Block
    font {
     bForceMonospaced=false
     bMaterialIsStencil=false
     fontMaterial=null
     fontObject=&Burbank
     letterSpacing=0
     monospacedWidth=1
     size=72
     skewAmount=0
     typefaceFontName=BurbankRegular-Black
     outlineSettings {
      bApplyOutlineToDropShadows=false
      bMiteredCorners=false
      bSeparateFillAlpha=false
      outlineMaterial=null
      outlineSize=3
      outlineColor=<rgba 0.9301 0.0037 0.0037 1>
     }
    }
    shadowColorAndOpacity=<rgba 0.099 0.0044 0 1>
    shadowOffset=<7 7>
   }
   slot {class=SizeBoxSlot}
  }
  TimeRemainingText_remainder {
   widgetClassPath=&CommonNumericTextBlock
   bIsVariable=true
   parent=Overlay_Timer
   widget=TimeRemainingText_remainder
   properties {
    currentNumericValue=888
    easeOutInterpolationExponent=1
    text=Text Block
    font {
     bForceMonospaced=false
     bMaterialIsStencil=false
     fontMaterial=null
     fontObject=&Burbank
     letterSpacing=0
     monospacedWidth=1
     size=24
     skewAmount=0
     typefaceFontName=BurbankRegular-Black
     outlineSettings {
      bApplyOutlineToDropShadows=false
      bMiteredCorners=false
      bSeparateFillAlpha=false
      outlineMaterial=null
      outlineSize=2
      outlineColor=<rgba 0.9301 0.0037 0.0037 1>
     }
    }
    renderTransform {angle=0, scale=<1 1>, shear=<0 0>, translation=<74 23>}
    shadowColorAndOpacity=<rgba 0.099 0.0044 0 1>
    shadowOffset=<6 6>
   }
   slot {class=OverlaySlot}
  }
 ]
}
