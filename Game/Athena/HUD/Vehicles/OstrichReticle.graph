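# Ostrich vehicle HUD reticle widget that shows the gunner's rocket loading and cooldown through a segmented-progress material.
# Events from OstrichReticleDecorator drive dynamic material parameters: loading sets fill and segment count, firing resets reload progress.
# Cooldown shows a dimmed, unsegmented reload state until complete, then restores empty segments.
# Segment count comes from the gunner weapon's max rockets, falling back to a default; shotgun reload and Construct events are empty.
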WidgetBlueprint /Game/Athena/HUD/Vehicles/OstrichReticle
graphs [
 EventGraph {nodes [
   Event Setup MID() {is=title K2Node_Event}
   > n1 = GetDynamicMaterial(self=Get Image_MissleReticle() -> (Image_MissleReticle:Image Object Reference))
   > n2 = Set ReticleMID(ReticleMID:Material Instance Dynamic Object Reference=@n1) -> (Output_Get:Material Instance Dynamic Object Reference)
   > SetScalarParameterValue(self=@n2, ParameterName=PercentFIllAmount, Value=0)
   > SetScalarParameterValue(self=@n2, ParameterName=Repeat, Value=GetGunnerRocketMax!(self) -> (ReturnValue:double))
   Event Rockets Fired() {is=title K2Node_Event}
   > SetScalarParameterValue(self=Get ReticleMID() -> (ReticleMID:Material Instance Dynamic Object Reference), ParameterName=PercentReloadAmount, Value=0)
   n3 = Event Rocket Loaded!() -> (OutputDelegate, then, Percent:float, Total:int=0) {is=title K2Node_Event}
   > n4 = GetDynamicMaterial(self=Get Image_MissleReticle())
   > SetScalarParameterValue(self=@n4, ParameterName=PercentFIllAmount, Value=@n3.Percent)
   > SetScalarParameterValue(self=@n4, ParameterName=Repeat, Value="To Float (Integer)"(InInt=@n3.Total))
   n5 = Event Rocket Cooldown!() -> (OutputDelegate, then, Percent:float) {is=title K2Node_Event}
   n6 = Get ReticleMID()
   > SetScalarParameterValue(self=@n6, ParameterName=PercentReloadAmount, Value=@n5.Percent)
   n7 = "float >= float"(A=@n5.Percent, B=1) {is=title K2Node_CallFunction}
   > SetScalarParameterValue(self=@n6, ParameterName=ReloadFade, Value=Select(Option 0:double=0.8, Option 1:double=1, Index=@n7) -> (ReturnValue:double))
   > SetScalarParameterValue(self=@n6, ParameterName=Repeat, Value=Select(Option 0=0, Option 1=GetGunnerRocketMax(), Index=@n7))
   > SetScalarParameterValue(self=@n6, ParameterName=PercentFIllAmount, Value=Select(Option 0=1, Option 1=0, Index=@n7))
   Event Construct()
   Event Shotgun Reload Start!() -> (OutputDelegate, then, Duration:float) {is=title K2Node_Event}
  ]}
 GetGunnerRocketMax {nodes [
   GetGunnerRocketMax() {is=title K2Node_FunctionEntry}
   n1 = Get OwningOstrichVehicle() -> (OwningOstrichVehicle:Fort Character Vehicle Ostrich Object Reference)
   > n2 = Is Valid(InputObject=@n1)
   n3 = Get Default Number Rockets() -> (Default Number Rockets:double)
   Return Node!(execute=@n2.Is Not Valid, ReturnValue:double=@n3)
   > n4 = GetGunnerWeapon!(execute, self:Fort Character Vehicle Ostrich Object Reference=@n1) -> (then, ReturnValue:Fort Weapon Ranged Ostrich Object Reference)
   > n5 = Is Valid(InputObject=@n4)
   Return Node(execute=@n5.Is Not Valid, ReturnValue=@n3)
   > Return Node(ReturnValue="To Float (Integer)"(InInt=GetMaxRockets!(self:Fort Weapon Ranged Ostrich Object Reference=@n4) -> (ReturnValue:int=0)))
  ]}
]
variables [
 Default Number Rockets {type=number, default=10}
 ReticleMID {type=MaterialInstanceDynamic}
]
widgets {
 info {inheritedWidgetCount=1, parentClass=/Script/FortniteUI.OstrichReticleDecorator, rootWidgetClass=&Overlay, widgetCount=3}
 widgets [
  Overlay_35 {widgetClassPath=&Overlay, widget=Overlay_35}
  Image_MissleReticle {
   widgetClassPath=&Image
   bIsVariable=true
   parent=Overlay_35
   widget=Image_MissleReticle
   properties {
    visibility=HitTestInvisible
    brush {
     bIsDynamicallyLoaded=false
     drawAs=Image
     imageType=NoImage
     mirroring=NoMirror
     resourceName=None
     resourceObject=/Game/Athena/HUD/reticles/M_UI_Reticle_SegmentedProgress_Inst
     tiling=NoTile
     imageSize=<128 128>
     margin=<ltrb 0 0 0 0>
     outlineSettings {
      bUseBrushTransparency=false
      roundingType=HalfHeightRadius
      width=0
      color {colorUseRule=UseColor_Specified, specifiedColor=<rgba 0 0 0 0>}
      cornerRadii=<0 0 0 0>
     }
     tintColor {colorUseRule=UseColor_Specified, specifiedColor=<rgba 1 1 1 1>}
     uVRegion {bIsValid=false, max=<0 0>, min=<0 0>}
    }
    renderTransform {angle=-106, scale=<1 1>, shear=<0 0>, translation=<0 0>}
   }
   slot {class=OverlaySlot, properties {horizontalAlignment=HAlign_Center, verticalAlignment=VAlign_Center}}
  }
  TutorialHighlight {widgetClassPath=&TutorialHighlight}
 ]
}
