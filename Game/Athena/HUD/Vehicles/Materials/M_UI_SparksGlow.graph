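# Additive UI material in the vehicle HUD folder that, by its name, draws a warm glow for sparks.
# A cone mask texture blends linearly from a dark orange-brown outer color to a bright amber core color.
# Both colors are exposed as vector parameters, and the graph has no animation or time input.
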Material /Game/Athena/HUD/Vehicles/Materials/M_UI_SparksGlow
graph {
 properties {blendMode=BLEND_Additive, materialDomain=MD_UI, pixelDepthOffsetMode=PDOM_Legacy, usedShadingModels=Default Lit}
 nodes [LinearInterpolate(
   A=VectorParameter() {parameterName=ColorOuter, defaultValue=<rgba 0.0865 0.0123 0.0009 1>}.RGB
   B=VectorParameter() {parameterName=ColorCore, defaultValue=<rgba 0.9387 0.6038 0.1912 1>}.RGB
   Alpha=TextureSample() {texture=&/T_UI_ConeMask}.RGB
  )]
}
