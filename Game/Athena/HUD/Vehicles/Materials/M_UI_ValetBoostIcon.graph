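# Translucent UI material for the Valet boost icon in the vehicle HUD.
# It treats the red channel of the grayscale T_UI_ValetBoost texture as a distance field, nudged up by a small bias and saturated.
# MF_UI_AntiAliasing thresholds that field at the midpoint for a smooth-edged shape, with the bias slightly fattening the icon.
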Material /Game/Athena/HUD/Vehicles/Materials/M_UI_ValetBoostIcon
graph {
 properties {
  bAutomaticallySetUsageInEditor=false
  blendMode=BLEND_Translucent
  materialDomain=MD_UI
  pixelDepthOffsetMode=PDOM_Legacy
  usedShadingModels=Default Lit
 }
 nodes [
  Constant() {r=1}
  n1 = TextureSample(UVs=TextureCoordinate()) {samplerType=SAMPLERTYPE_LinearGrayscale, texture=&/T_UI_ValetBoost}
  MaterialFunctionCall(DistanceField=Saturate(None=Add(A=@n1.R) {constB=0.09}), Threshold=Constant() {r=0.5}) {materialFunction=&MF_UI_AntiAliasing}
 ]
}
